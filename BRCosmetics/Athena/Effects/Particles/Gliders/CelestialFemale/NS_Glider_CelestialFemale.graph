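# Niagara effect for the CelestialFemale glider cosmetic: anim-trail ribbons, a pulsing heart sprite, rotating orbit meshes and twinkling stars.
# Two skeletal-mesh ribbons spawn when fully deployed and either ForwardDot passes a threshold or in front end, where added velocity makes them stream.
# Stars spawn at both wing offsets and around the heart.
# ForwardDot also switches the moving stars' spawn rate, lifetime and drift velocity.
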NiagaraSystem /BRCosmetics/Athena/Effects/Particles/Gliders/CelestialFemale/NS_Glider_CelestialFemale
dependencies {usedDataInterfaces [&NiagaraDataInterfaceCurve], usedDynamicInputs [], usedModules [&SystemState]}
emitters {
 FN_AnimTrail {
  emitterSpawnScript [&FN_RibbonIDOffsetUtility()]
  emitterUpdateScript [
   &EmitterState()
   &FN_SpawnAnimTrail(
    Enable:NiagaraBool=&Boolean_AndOperation(
     Bool1=$User.bIsFullyDeployed
     Bool2=&Boolean_OrOperation(
      Bool1=&SetBoolByFloatComparison(A=$User.ForwardDot, B=0.2, Comparison Type=::NewEnumerator2)
      Bool2=$User.bisFrontEnd
     )
    )
    Number of Anim Trails=2
   )
  ]
  particleSpawnScript [
   &FN_InitializeAnimTrailBoneIndecies()
   &FN_InitializeAnimTrailBoneIndecies(AnimTrail Selection=1, First Bone or Socket Index=2, Second Bone or Socket Index=3) {module=FN_InitializeAnimTrailBoneIndecies001}
   &FN_InitializeAnimTrail(AnimTrailWidth=1.7, Color=<rgba 1 1 1 0.75>, Lifetime=0.25)
   &AddVelocity(Velocity=&Multiply_Vector(
     A=<0 800 0>
     B=&VectorFromFloat(Value=&MakeCustomFloatFromBool(Bool=$User.bisFrontEnd, True Float=1, False Float=0))
    ))
   &DynamicMaterialParameters(
    Index 0 Param 2=1
    Index 0 Param 3=1
    Index 0 Param 4=1
    Module.Param0WriteEnabled=true
    Module.Param1WriteEnabled=true
    Module.Param2WriteEnabled=true
    Module.Param3WriteEnabled=true
   )
  ]
  particleUpdateScript [
   &SolveForcesAndVelocity(Clamp Velocity=false, Limit Acceleration=false)
   &FN_UpdateAnimTrail(Locked Particle RibbonWidth=1, SkeletalMesh=$User.SkeletalMesh) {module=FN_UpdateAnimTrail001}
   /SetVariables(Particles.RibbonU0Override=$Particles.NormalizedAge) {isSetParameters=true}
   &ScaleColor(
    Scale Alpha=&FloatFromCurve(FloatCurve=DataInterface() {propertyValues {Curve {keys [
         {arriveTangent=27.8125, interpMode=RCIM_Cubic, leaveTangent=27.8125, tangentMode=RCTM_User, time=0, value=0}
         <at 0.05 1 RCIM_Cubic>
         <at 0.5 0 RCIM_Cubic>
        ]}}})
    ScaleA=true
    ScaleRGB=true
   )
  ]
  rendererClasses [&NiagaraRibbonRendererProperties]
  renderers [{rendererClass=&NiagaraRibbonRendererProperties}]
 }
 Heart {
  emitterUpdateScript [
   &EmitterState(Inactive Response=::NewEnumerator1, Life Cycle Mode=::NewEnumerator1, Loop Duration=1.8, UseLoopDelay=false)
   &SpawnBurst_Instantaneous(Spawn Count=1, Use Spawn Probability=false)
  ]
  particleSpawnScript [
   /Niagara/Modules/Spawn/Initialization/InitializeParticle(
    Color=<rgba 0.7487 0.3875 1.5 0.4>
    Lifetime=2
    Module.Write Color=true
    Module.Write Lifetime=true
    Module.Write Mass=true
    Module.Write Material Random=true
    Module.Write Position=false
    Module.Write Scale=false
    Module.Write SpriteRotation=false
    Module.Write SpriteSize=true
    Position=&SimulationPosition()
    Sprite Size=&Vector2DFromFloat(Value=&UniformRangedFloat(Minimum=25, Maximum=40))
   )
   &CameraOffset(Camera Offset Amount=75)
   &SystemLocation(MaskSpawnGroup=false, Offset=<25 0 105>)
  ]
  particleUpdateScript [
   &ParticleState()
   &ScaleColor(
    Scale Alpha=&FloatFromCurve(FloatCurve=DataInterface() {propertyValues {Curve {keys [
         <at 0.0051 0.1919 RCIM_Cubic>
         {arriveTangent=15.763, interpMode=RCIM_Cubic, leaveTangent=15.7629, tangentMode=RCTM_User, time=0.1, value=0.85}
         {arriveTangent=-6.6929, interpMode=RCIM_Cubic, leaveTangent=-6.6929, tangentMode=RCTM_User, time=0.15, value=1.2}
         {interpMode=RCIM_Cubic, tangentMode=RCTM_User, time=0.3, value=0.65}
         {arriveTangent=1.1254, interpMode=RCIM_Cubic, leaveTangent=1.1254, tangentMode=RCTM_User, time=0.81, value=0.8}
         {arriveTangent=8.7241, interpMode=RCIM_Cubic, leaveTangent=8.7241, tangentMode=RCTM_User, time=0.91, value=1.15}
         <at 1 0.2 RCIM_Cubic>
        ]}}})
    ScaleA=true
    ScaleRGB=true
   )
  ]
  rendererClasses [&NiagaraSpriteRendererProperties]
  renderers [{rendererClass=&NiagaraSpriteRendererProperties}]
 }
 Orbit {
  emitterUpdateScript [
   &EmitterState(
    Inactive Response=::NewEnumerator1
    Life Cycle Mode=::NewEnumerator1
    Loop Duration=&UniformRangedFloat(Minimum=3, Maximum=6)
    Recalculate Duration Each Loop=true
    UseLoopDelay=false
   )
   &SpawnBurst_Instantaneous(Spawn Count=1, Use Spawn Probability=false)
  ]
  particleSpawnScript [
   &InitializeParticle(
    Color=&LerpLinearColors(StartColor=<rgba 0.0208 0.7213 2 1>, EndColor=<rgba 0.1832 2 1.8805 1>, LerpFactor=&UniformAOrBFloat(A=0))
    Lifetime=5
    Mesh Scale=&VectorFromFloat(Value=&UniformRangedFloat(Minimum=1.5, Maximum=1.5))
    Module.Write Color=true
    Module.Write Lifetime=true
    Module.Write Mass=true
    Module.Write Material Random=true
    Module.Write Position=false
    Module.Write Scale=true
    Module.Write SpriteRotation=false
    Module.Write SpriteSize=false
    Position=&SimulationPosition()
    Sprite Size=<5 5>
   )
   &DynamicMaterialParameters(
    Index 0 Param 1=&UniformRangedFloat(Minimum=0.3, Maximum=0.7)
    Module.Param0WriteEnabled=true
    Module.Param1WriteEnabled=false
    Module.Param2WriteEnabled=false
    Module.Param3WriteEnabled=false
   )
   &InitialMeshOrientation(
    Orientation Vector=&UniformRangedVector(Minimum=<-0.25 0 1>, Maximum=<0.25 0 1>)
    Use Orientation Vector=true
    Use Rotation Vector=true
   ) {module=InitialMeshOrientation001}
   &SystemLocation(MaskSpawnGroup=false, Offset=<0 0 55>)
  ]
  particleUpdateScript [
   &ParticleState()
   &SolveForcesAndVelocity(Clamp Velocity=false, Limit Acceleration=false)
   &DynamicMaterialParameters(
    Index 0 Param 2=&FloatFromCurve(FloatCurve=DataInterface() {propertyValues {Curve {keys [
         {arriveTangent=-6.5008, interpMode=RCIM_Cubic, leaveTangent=-6.5008, tangentMode=RCTM_User, time=0, value=2}
         <at 0.5 0.8 RCIM_Cubic>
         <at 1 2 RCIM_Cubic>
        ]}}})
    Index 0 Param 3=&FloatFromCurve(FloatCurve=DataInterface() {propertyValues {Curve {keys [{arriveTangent=1.7971, interpMode=RCIM_Cubic, leaveTangent=1.7971, tangentMode=RCTM_User, time=0, value=0}, <at 1 1 RCIM_Cubic>]}}})
    Index 0 Param 4=&FloatFromCurve(FloatCurve=DataInterface() {propertyValues {Curve {keys [
         <at 0 0 RCIM_Cubic>
         {interpMode=RCIM_Cubic, tangentMode=RCTM_User, time=0.2, value=0.5}
         {interpMode=RCIM_Cubic, tangentMode=RCTM_User, time=0.7, value=0.5}
         {arriveTangent=-4.058, interpMode=RCIM_Cubic, leaveTangent=-4.058, tangentMode=RCTM_User, time=1, value=0}
        ]}}})
    Module.Param0WriteEnabled=false
    Module.Param1WriteEnabled=true
    Module.Param2WriteEnabled=true
    Module.Param3WriteEnabled=true
   ) {module=DynamicMaterialParameters001}
   &UpdateMeshOrientation(
    Rotation Coordinate Space=::NewEnumerator3
    Rotation Rate=&FloatFromCurve(FloatCurve=DataInterface() {propertyValues {Curve {keys [<at 0.2 1 RCIM_Cubic>, {arriveTangent=-2.6679, interpMode=RCIM_Cubic, leaveTangent=-2.6679, tangentMode=RCTM_User, time=1, value=0}]}}})
    Rotation Vector=&UniformRangedVector(Minimum=<-0.6 -0.03 0>, Maximum=<-0.8 0.03 0>)
   ) {module=UpdateMeshOrientation001}
   &ScaleColor(
    Scale Alpha=&FloatFromCurve(FloatCurve=DataInterface() {propertyValues {Curve {keys [
         <at 0 0 RCIM_Cubic>
         {interpMode=RCIM_Cubic, tangentMode=RCTM_User, time=0.2, value=1}
         {interpMode=RCIM_Cubic, tangentMode=RCTM_User, time=0.5, value=1}
         <at 1 0 RCIM_Cubic>
        ]}}})
    ScaleA=true
    ScaleRGB=true
   )
  ]
  rendererClasses [&NiagaraMeshRendererProperties]
  renderers [{rendererClass=&NiagaraMeshRendererProperties, rendererIndex=0}]
 }
 Stars_Moving {
  emitterUpdateScript [
   &EmitterState(Life Cycle Mode=::NewEnumerator1, Loop Duration=1, UseLoopDelay=false)
   &SpawnRate(
    SpawnRate=&Lerp_Float(
     A=1
     B=5
     Alpha=&MakeCustomFloatFromBool(
      Bool=&SetBoolByFloatComparison(A=$User.ForwardDot, B=0.2, Comparison Type=::NewEnumerator2)
      True Float=1
      False Float=0
     )
    )
    Use Spawn Probability=false
   )
  ]
  particleSpawnScript [
   &InitializeParticle(
    Color=&UniformRangedLinearColor(Minimum=<rgba 0.1703 0.4623 1 0.7>, Maximum=<rgba 0.1666 0.9901 1 0.7>)
    Lifetime=&Lerp_Float(
     A=&UniformRangedFloat(Minimum=0.5, Maximum=1)
     B=&UniformRangedFloat(Minimum=0.5, Maximum=2)
     Alpha=&MakeCustomFloatFromBool(
      Bool=&SetBoolByFloatComparison(A=$User.ForwardDot, B=0.5, Comparison Type=::NewEnumerator4)
      True Float=0
      False Float=1
     )
    )
    Module.Write Color=true
    Module.Write Lifetime=true
    Module.Write Mass=true
    Module.Write Material Random=true
    Module.Write Position=false
    Module.Write Scale=false
    Module.Write SpriteRotation=false
    Module.Write SpriteSize=true
    Position=&SimulationPosition()
    Sprite Size=&Vector2DFromFloat(Value=&UniformRangedFloat(Minimum=10, Maximum=70))
   )
   &DynamicMaterialParameters(
    Index 0 Param 1=&UniformRangedFloat(Minimum=0.4, Maximum=0.7)
    Index 0 Param 2=&UniformRangedFloat(Minimum=0.8, Maximum=1.55)
    Index 0 Param 3=1
    Index 0 Param 4=1
    Module.Param0WriteEnabled=true
    Module.Param1WriteEnabled=true
    Module.Param2WriteEnabled=true
    Module.Param3WriteEnabled=true
   )
   &CylinderLocation(
    Cylinder Height=1
    Cylinder Radius=50
    Hemisphere Y=true
    MaskSpawnGroup=false
    Module.Enable Lathe Profile=false
    Module.Override Randomness=false
    Module.Surface Only=false
    Offset=&Lerp_Vector(A=<-115 -20 -10>, B=<115 -20 -10>, Alpha=&UniformAOrBFloat(A=0))
   )
   &AddVelocity(Velocity=&Multiply_Vector(A=$Output.CylinderLocation.CylinderNormal, B=&VectorFromFloat(Value=0.5)))
   &CameraOffset(Camera Offset Amount=25)
  ]
  particleUpdateScript [
   &ParticleState()
   &AddVelocity(Velocity=&MakeVector(
     X=0
     Y=&FloatFromCurve(
      FloatCurve=DataInterface() {propertyValues {Curve {keys [<at 0 0 RCIM_Cubic>, {arriveTangent=5.6885, interpMode=RCIM_Cubic, leaveTangent=5.6885, tangentMode=RCTM_User, time=0.6, value=1}]}}}
      Scale Curve=&Lerp_Float(
       A=&Lerp_Float(
        A=-75
        B=75
        Alpha=&MakeCustomFloatFromBool(
         Bool=&SetBoolByFloatComparison(A=$User.ForwardDot, B=0.2, Comparison Type=::NewEnumerator2)
         True Float=1
         False Float=0
        )
       )
       B=50
       Alpha=&MakeCustomFloatFromBool(
        Bool=&SetBoolByFloatComparison(A=$User.ForwardDot, B=0.2, Comparison Type=::NewEnumerator2)
        True Float=0
        False Float=1
       )
      )
     )
     Z=0
    )) {module=AddVelocity002}
   &Drag(
    Drag=&FloatFromCurve(FloatCurve=DataInterface() {propertyValues {Curve {keys [<at 0.75 1>,<at 1 0>]}}}, Scale Curve=5)
    Use Linear Drag=true
    Use Rotational Drag=false
   )
   &SolveForcesAndVelocity(Clamp Velocity=false, Limit Acceleration=false)
   &ScaleColor(
    Scale Alpha=&FloatFromCurve(FloatCurve=DataInterface() {propertyValues {Curve {keys [
         <at 0 0 RCIM_Cubic>
         {interpMode=RCIM_Cubic, tangentMode=RCTM_User, time=0.2, value=1}
         {interpMode=RCIM_Cubic, tangentMode=RCTM_User, time=0.8, value=1}
         <at 1 0 RCIM_Cubic>
        ]}}})
    ScaleA=true
    ScaleRGB=true
   )
   &ScaleSpriteSize(Scale Factor=&Vector2DFromFloat(Value=&FloatFromCurve(FloatCurve=DataInterface() {propertyValues {Curve {keys [
          <at 0 0 RCIM_Cubic>
          {arriveTangent=2.0455, interpMode=RCIM_Cubic, leaveTangent=2.0455, time=0.1, value=1.15}
          {arriveTangent=-0.4762, interpMode=RCIM_Cubic, leaveTangent=-0.4762, time=0.22, value=0.45}
          {arriveTangent=-1.082, interpMode=RCIM_Cubic, leaveTangent=-1.082, time=0.31, value=1.05}
          {arriveTangent=-0.0758, interpMode=RCIM_Cubic, leaveTangent=-0.0758, time=0.525, value=0.12}
          {arriveTangent=0.4561, interpMode=RCIM_Cubic, leaveTangent=0.4561, time=0.64, value=1.025}
          {arriveTangent=-2.8472, interpMode=RCIM_Cubic, leaveTangent=-2.8472, time=0.81, value=0.25}
          <at 1 0 RCIM_Cubic>
         ]}}})))
  ]
  rendererClasses [&NiagaraSpriteRendererProperties]
  renderers [{rendererClass=&NiagaraSpriteRendererProperties}]
 }
 Stars_Wings {
  emitterUpdateScript [
   &EmitterState(Inactive Response=::NewEnumerator1, Life Cycle Mode=::NewEnumerator1, Loop Duration=1, UseLoopDelay=false)
   &SpawnRate(SpawnRate=2, Use Spawn Probability=false)
  ]
  particleSpawnScript [
   &InitializeParticle(
    Color=&UniformRangedLinearColor(Minimum=<rgba 0.1703 0.4623 1 0.7>, Maximum=<rgba 0.1666 0.9901 1 0.7>)
    Lifetime=&UniformRangedFloat(Minimum=0.5, Maximum=3)
    Module.Write Color=true
    Module.Write Lifetime=true
    Module.Write Mass=true
    Module.Write Material Random=true
    Module.Write Position=false
    Module.Write Scale=false
    Module.Write SpriteRotation=false
    Module.Write SpriteSize=true
    Position=&SimulationPosition()
    Sprite Size=&Vector2DFromFloat(Value=&UniformRangedFloat(Minimum=7, Maximum=15))
   )
   &DynamicMaterialParameters(
    Index 0 Param 1=&UniformRangedFloat(Minimum=0.4, Maximum=0.7)
    Index 0 Param 2=&UniformRangedFloat(Minimum=0.8, Maximum=1.55)
    Index 0 Param 3=1
    Index 0 Param 4=1
    Module.Param0WriteEnabled=true
    Module.Param1WriteEnabled=true
    Module.Param2WriteEnabled=true
    Module.Param3WriteEnabled=true
   )
   &CylinderLocation(
    Cylinder Height=1
    Cylinder Radius=45
    Hemisphere Y=true
    MaskSpawnGroup=false
    Module.Enable Lathe Profile=false
    Module.Override Randomness=false
    Module.Surface Only=false
    Offset=&Lerp_Vector(A=<-115 -20 -10>, B=<115 -20 -10>, Alpha=&UniformAOrBFloat(A=0))
   )
   &CameraOffset(Camera Offset Amount=50)
  ]
  particleUpdateScript [
   &ParticleState()
   &ScaleColor(
    Scale Alpha=&FloatFromCurve(FloatCurve=DataInterface() {propertyValues {Curve {keys [
         <at 0 0 RCIM_Cubic>
         {interpMode=RCIM_Cubic, tangentMode=RCTM_User, time=0.2, value=1}
         {interpMode=RCIM_Cubic, tangentMode=RCTM_User, time=0.8, value=1}
         <at 1 0 RCIM_Cubic>
        ]}}})
    ScaleA=true
    ScaleRGB=true
   )
  ]
  rendererClasses [&NiagaraSpriteRendererProperties]
  renderers [{rendererClass=&NiagaraSpriteRendererProperties}]
 }
 Stars_Wings001 {
  emitterUpdateScript [
   &EmitterState(Inactive Response=::NewEnumerator1, Life Cycle Mode=::NewEnumerator1, Loop Duration=1.8, UseLoopDelay=false)
   &SpawnRate(Use Spawn Probability=false)
  ]
  particleSpawnScript [
   &InitializeParticle(
    Color=&UniformRangedLinearColor(Minimum=<rgba 0.1703 0.4623 1 0.7>, Maximum=<rgba 0.1666 0.9901 1 0.7>)
    Lifetime=&UniformRangedFloat(Minimum=2, Maximum=3)
    Module.Write Color=true
    Module.Write Lifetime=true
    Module.Write Mass=true
    Module.Write Material Random=true
    Module.Write Position=false
    Module.Write Scale=false
    Module.Write SpriteRotation=false
    Module.Write SpriteSize=true
    Position=&SimulationPosition()
    Sprite Size=&Vector2DFromFloat(Value=&UniformRangedFloat(Minimum=7, Maximum=15))
   )
   &DynamicMaterialParameters(
    Index 0 Param 1=&UniformRangedFloat(Minimum=0.4, Maximum=0.7)
    Index 0 Param 2=&UniformRangedFloat(Minimum=0.8, Maximum=1.55)
    Index 0 Param 3=1
    Index 0 Param 4=1
    Module.Param0WriteEnabled=true
    Module.Param1WriteEnabled=true
    Module.Param2WriteEnabled=true
    Module.Param3WriteEnabled=true
   )
   &CylinderLocation(
    Cylinder Height=5
    Cylinder Radius=50
    Lathe Profile=DataInterface() {propertyValues {Curve {keys [
        {arriveTangent=1.8498, interpMode=RCIM_Cubic, leaveTangent=1.8498, tangentMode=RCTM_User, time=0, value=0.451}
        {arriveTangent=0.049, interpMode=RCIM_Cubic, leaveTangent=0.049, time=0.5, value=1}
        {arriveTangent=-1.7498, interpMode=RCIM_Cubic, leaveTangent=-1.7498, tangentMode=RCTM_User, time=1, value=0.5}
       ]}}}
    MaskSpawnGroup=false
    Module.Enable Lathe Profile=true
    Module.Override Randomness=false
    Module.Surface Only=false
    Offset=<0 0 105>
   )
   &CameraOffset(Camera Offset Amount=&UniformRangedFloat(Minimum=50, Maximum=100))
  ]
  particleUpdateScript [
   &ParticleState()
   &ScaleColor(
    Scale Alpha=&FloatFromCurve(FloatCurve=DataInterface() {propertyValues {Curve {keys [
         <at 0 0 RCIM_Cubic>
         {interpMode=RCIM_Cubic, tangentMode=RCTM_User, time=0.2, value=1}
         {interpMode=RCIM_Cubic, tangentMode=RCTM_User, time=0.8, value=1}
         <at 1 0 RCIM_Cubic>
        ]}}})
    ScaleA=true
    ScaleRGB=true
   )
  ]
  rendererClasses [&NiagaraSpriteRendererProperties]
  renderers [{rendererClass=&NiagaraSpriteRendererProperties}]
 }
}
summary {
 systemName=NS_Glider_CelestialFemale
 emitters [
  {emitterName=Stars_Moving, rendererClasses [&NiagaraSpriteRendererProperties]}
  {emitterName=Stars_Wings, rendererClasses [&NiagaraSpriteRendererProperties]}
  {emitterName=Heart, rendererClasses [&NiagaraSpriteRendererProperties]}
  {emitterName=Stars_Wings001, rendererClasses [&NiagaraSpriteRendererProperties]}
  {emitterName=Orbit, rendererClasses [&NiagaraMeshRendererProperties]}
  {emitterName=FN_AnimTrail, rendererClasses [&NiagaraRibbonRendererProperties]}
 ]
 userVariables [
  User.bisFrontEnd {type=NiagaraBool, defaultValue=true}
  User.bIsFullyDeployed {type=NiagaraBool, defaultValue=true}
  User.ForwardDot {type=NiagaraFloat, defaultValue=0, description=""}
  User.RibbonIDOffset {type=NiagaraInt32}
  User.RightDot {type=NiagaraFloat, defaultValue=0, description=""}
  User.SkeletalMesh {
   type=NiagaraDataInterfaceSkeletalMesh
   defaultValue {
    struct=NiagaraExt_VariableValue_DataInterface
    value {dataInterface=&self:NiagaraDataInterfaceSkeletalMesh_0}
   }
  }
 ]
}
